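; Kiosk quest prop for the Season 24 narrative quest (Part 04, Q03) that sets up the starting layout of an arcade-style screen.
; After the parent construction script runs, it writes scalar parameters to the screen's dynamic material instance: green dot start from a variable, plus fixed end point and four obstacle indices.
; The event graph holds only disabled stubs, so any runtime behavior comes from the parent class.
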(event EventBeginPlay
  (|CallParentFunction) ; This node is disabled and will not be called.\nDrag off pins to build functionality.
)

(event Collision|EventActorBeginOverlap
  (params (OtherActor :type Actor%20Object%20Reference))
  (|CallParentFunction OtherActor) ; This node is disabled and will not be called.\nDrag off pins to build functionality.
)

(event EventTick
  (params (DeltaSeconds :type Float%20%28single-precision%29))
  (|CallParentFunction DeltaSeconds) ; This node is disabled and will not be called.\nDrag off pins to build functionality.
)

(fn ConstructionScript
  (bind _screenmid (Variables|Default|GetScreenMID))
  ;[0 0 0] Set initial arcade values for this screen
  (|CallParentFunction)
  (Rendering|Material|SetScalarParameterValue _screenmid "Green dot index" (Variables|Default|GetGreenDotStart) :!class /Script/Engine.MaterialInstanceDynamic)
  (Rendering|Material|SetScalarParameterValue _screenmid "End Point Index" 1.0 :!class /Script/Engine.MaterialInstanceDynamic)
  (Rendering|Material|SetScalarParameterValue _screenmid "obstacle index" 2.0 :!class /Script/Engine.MaterialInstanceDynamic)
  (Rendering|Material|SetScalarParameterValue _screenmid "obstacle2 index" 7.0 :!class /Script/Engine.MaterialInstanceDynamic)
  (Rendering|Material|SetScalarParameterValue _screenmid "obstacle3 index" 9.0 :!class /Script/Engine.MaterialInstanceDynamic)
  (Rendering|Material|SetScalarParameterValue _screenmid "obstacle4 index" 1e+01 :!class /Script/Engine.MaterialInstanceDynamic))
